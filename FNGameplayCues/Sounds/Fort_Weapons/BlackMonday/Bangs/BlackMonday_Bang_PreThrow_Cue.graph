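# Sound cue for the pre-throw sound of the BlackMonday Bang weapon.
# A Mixer layers the TimedBomb pre-throw wave at full input volume with a BlackMonday Bang pre-throw wave at reduced volume, so both play together.
# It uses close-range prop attenuation and the player foley sound class, with a concurrency override that prevents new instances once the limit is reached.
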SoundCue /FNGameplayCues/Sounds/Fort_Weapons/BlackMonday/Bangs/BlackMonday_Bang_PreThrow_Cue
root=Mixer_0
nodes [
 {id=Mixer_0, children [WavePlayer_1,WavePlayer_0], properties {inputVolume [0.5,1]}}
 {id=WavePlayer_0, properties {soundWaveAssetPtr=/FNGameplayCues/Sounds/Fort_Gadgets/TimedBomb/TimedBomb_PreThrow_01}}
 {id=WavePlayer_1, properties {soundWaveAssetPtr=/Game/Sounds/Fort_Weapons/BlackMonday/Bangs/BlackMonday_Bang_PreThrow_02}}
]
properties {
 attenuationSettings=/Game/Sounds/Attenuation/Fort_Prop_Close_Att
 bOverrideConcurrency=true
 maxDistance=1800
 soundClassObject=&Fort_Player_Foley
 volumeMultiplier=1
 concurrencyOverrides {resolutionRule=PreventNew}
}
